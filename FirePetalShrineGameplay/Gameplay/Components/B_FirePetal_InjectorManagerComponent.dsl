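; B_FirePetal_InjectorManagerComponent is a FirePetal Shrine gameplay component that manages a level injection over its lifetime.
; On BeginPlay it reads a scalable-float hotfix that chooses between requesting injection immediately or delaying it.
; The delayed path waits on the authority match-ready async action's OnReady before requesting injection, a delay the hotfix names as until warmup phase.
; On EndPlay it removes the injection.
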(event EventBeginPlay
  (if (ScalableFloat|GetValueasBool (Variables|Hotfixes|GetHF_DelayInjectionTilWarmupPhase))
    (Class|FortAsyncActionAuthorityMatchReady|MatchReadyAsync
      (:OnReady
        (CallFunction|RequestLevelInjection)))
    (else
      (CallFunction|RequestLevelInjection))))

(event EventEndPlay
  (params (EndPlayReason :type EEndPlayReason%20Enum :default "Destroyed"))
  (CallFunction|RemoveLevelInjection))
