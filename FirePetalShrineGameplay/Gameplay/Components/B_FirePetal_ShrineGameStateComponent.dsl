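; Game state component that spawns Fire Petal shrine building actors across the map at locations chosen by an EQS query.
; BeginPlay triggers the query once the playlist is ready, or via a debug flag in PIE, and a guard limits it to one run.
; Scalable floats set grid size and spacing, and each result location gets a randomly yawed shrine.
; Spawned shrines are tracked so DisableShrines can destroy them.
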(event EventBeginPlay
  ;[0.7899 0.1002 0.2833] Wait for playlist to be ready before we do EQS spawn
  (if (Editor|InEditororPIE)
    (if (Variables|Default|GetDebugRunOnBeginPlay?)
      _)
    (else
      (Class|FortAsyncActionCurrentPlaylistReady|PlaylistDataReadyAsync
        (:OnReady
          _))))
  (CallFunction|DoEQS))

(event Custom|DoEQS
  ;[0.7899 0.1002 0.2833] Run the EQS Query
  (if (Variables|Default|GetHasRunEQS?)
    (Development|PrintString "Shrine EQS has already run!" true true "(R=0.000000,G=0.660000,B=1.000000,A=1.000000)" 4.0)
    (else
      (bind _returnvalue (AI|EQS|RunEQSQuery (Variables|Env|GetEQSQueryTemplate) self "AllMatching"))
      (AI|EQS|SetNamedParam _returnvalue "SimpleGrid.GridSize" (ScalableFloat|GetValueatLevel (Variables|Env|GetHF_EQSRadius)))
      (AI|EQS|SetNamedParam _returnvalue "SimpleGrid.SpaceBetween" (ScalableFloat|GetValueatLevel (Variables|Env|GetHF_EQSSpacing)))
      (AI|EQS|BindEventtoOnQueryFinished _returnvalue (EventDispatchers|CreateEvent :!event Custom|EQSQueryFinished))
      (Variables|Default|SetHasRunEQS? true)
      (Development|PrintString "Running Shrine EQS..." true true "(R=0.789931,G=0.100244,B=0.283317,A=1.000000)"))))

(event Custom|EQSQueryFinished
  (params (QueryInstance :type EQS%20Query%20Instance%20Object%20Reference) (QueryStatus :type EEnvQueryStatus%20Enum :default "Processing"))
  ;[0.7899 0.1002 0.2833] EQS finished -> start spawning shrines
  (bind (_resultlocations _returnvalue) (AI|EQS|GetQueryResultsasLocations QueryInstance))
  (if _returnvalue
    (Development|PrintString "Shrine EQS successful!" true true "(R=0.121770,G=1.000000,B=0.000000,A=1.000000)" 5.0)
    (CallFunction|SpawnAllShrines _resultlocations)
    (else
      (Development|PrintString "Shrine EQS failed :(" true true "(R=0.731250,G=0.144727,B=0.144727,A=1.000000)" 5.0))))

(fn DisableShrines
  (for _array_element (Variables|Default|GetSpawnedShrines)
    (Actor|DestroyActor _array_element)))

(fn SpawnAllShrines
  (params (Locations :type Array%20of%20Vectors))
  (for _array_element Locations
    (CallFunction|SpawnShrine _array_element))
  (return))

(fn SpawnShrine
  (params (Location :type Vector))
  (bind _returnvalue (Game|SpawnBuildingActorfromClass (Variables|Default|GetShrineBGAClass) (Math|Transform|MakeTransform Location (Math|Rotator|MakeRotator 0.0 0.0 (Math|Random|RandomFloatinRange 0.0 3.6e+02))) 0 false 0 0))
  (Utilities|Array|AddUnique (Variables|Default|GetSpawnedShrines) _returnvalue)
  (Development|PrintText (Utilities|Text|FormatText "NSLOCTEXT(\"[496C037BEC01523C060C82C07E5D2E72]\", \"AE30C3C0480E607DA404B4B354F47F5D\", \"Shrine spawned at {x}!\")" (Utilities|String|ToString(Vector) (Variables|GetLocation))) true true "(R=0.000000,G=1.000000,B=0.048047,A=1.000000)" 1.0))
